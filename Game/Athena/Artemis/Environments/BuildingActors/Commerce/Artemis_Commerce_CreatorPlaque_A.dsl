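; Interactable creator plaque building actor that opens the Creative Support a Creator panel for the local player.
; It is always interactable, shows a configurable prompt, and on local interact runs the widget creation event.
; That event checks runtime kill switches, cosmetic-event execution and cursor mode, then asynchronously creates the affiliate panel by class path.
; The panel is initialized with the plaque itself, which exposes affiliate and source names through GetSACDetails.
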(event Custom|Create%20Support%20a%20Creator%20Widget
  ;[1 0.6509 0 0.25] Creates and shows the Support a Creator panel when activated. Called from OnLocalInteract.
  (if (not (or (Variables|/Script/FortniteGame.FortRuntimeOptions|GetbDisableAffiliateFeature (Options|GetRuntimeOptions)) (Variables|/Script/FortniteGame.FortRuntimeOptions|GetbDisableCreativeSACFeature (Options|GetRuntimeOptions))))
    (Networking|CanExecuteCosmeticEvents
      (:True
        (if (not (HUD|IsinCursorMode (BlueprintContext|GetFortHUDContext)))
          (bind _userwidget (Class|FortAsyncActionCreateWidgetAsync|CreateWidgetAsync "/Game/Athena/UI/InGame/Creative/CreativeAffiliatePanel.CreativeAffiliatePanel_C" (Fortnite|GetFirstLocalFortPlayerController true))
            (:OnComplete
              (Init|InitfromObject(Message) _userwidget self))))))))

(fn BlueprintCanInteract
  (params (InteractingPawn :type Fort%20Pawn%20Object%20Reference) (InteractionBeingAttempted :type EInteractionBeingAttempted%20Enum :default "FirstInteraction") (InteractionType :type TInteractionType%20Enum :default "IT_NoInteraction"))
  (return true))

(fn BlueprintGetInteractionString
  (params (InteractingPawn :type Fort%20Pawn%20Object%20Reference) (InteractionBeingAttempted :type EInteractionBeingAttempted%20Enum :default "FirstInteraction"))
  (return (Variables|SupportaCreatorDetails|GetInteraction%20Text)))

(fn BlueprintOnLocalInteract
  (params (InteractingPawn :type Fort%20Player%20Pawn%20Object%20Reference) (InteractionBeingAttempted :type EInteractionBeingAttempted%20Enum :default "FirstInteraction"))
  (CallFunction|CreateSupportACreatorWidget))

(fn GetSACDetails
  (return (Variables|SupportaCreatorDetails|GetAffiliate%20Name) (Variables|SupportaCreatorDetails|GetSource%20Name) false))

(fn ConstructionScript)
